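# Sound cue for the third tentacle attack sound in Section 02 of the PistolCanary event.
# A single wave player plays one tentacle attack wave at slightly reduced volume, with no randomization or mixing.
# It uses the tentacle-tip attenuation settings with a very large max distance and routes to the Fort_Events sound class.
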SoundCue /PistolCanary_Event_Audio/Assets/Section02/tacaud1500/SC_PistolCanary_Section02_tacaud1500_Attack03
root=WavePlayer_0
nodes [{id=WavePlayer_0, properties {soundWaveAssetPtr=/PistolCanary_Event_Audio/Assets/Section02/wav/PistolCanary_Tentacle_Attk_1000_03}}]
properties {
 attenuationSettings=/PistolCanary_Audio/Assets/Settings/ATT_PistolCanary_Tentacle_Tip
 maxDistance=100000
 soundClassObject=&Fort_Events
 volumeMultiplier=0.8
}
